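# Physics asset for the Durian Tier 1 vehicle body skeletal mesh, approximating its bones with one capsule each.
# Glass, engine panels, rear fenders and the rear-left shock constrain to the central body capsule.
# Wheel discs, rear brake pads and front-right steering form separate axle or steering chains, with no constraints linking them to the body.
# All constraints share identical swing and twist limit angles, and left-right coverage is asymmetric.
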PhysicsAsset /VehicleCosmetics/Bodies/durianTier1/Meshes/SK_Body_DurianT1_PhysicsAsset
bodies {
 axle_pivot_bk_l [{shapeType=Capsule, length=19.7969, radius=18.8591, center=<1.7101 -13.18 5.3892>, rotation=<rot 19.3487 -90 25.7039>}]
 axle_pivot_bk_r [{shapeType=Capsule, length=19.7969, radius=18.8591, center=<1.7101 13.18 5.3892>, rotation=<rot -19.3488 -90 25.7039>}]
 axle_pivot_fr_r [{shapeType=Capsule, length=23.1075, radius=18.1786, center=<-1.6682 17.811 2.7984>, rotation=<rot 0 -157.2849 -49.0394>}]
 back_fender_l [{shapeType=Capsule, length=95.6999, radius=39.8735, center=<-37.5077 -39.0424 -12.6319>, rotation=<rot 0 -80.2623 -88.854>}]
 back_fender_r [{shapeType=Capsule, length=95.7, radius=39.8735, center=<-37.5077 39.0424 -12.6319>, rotation=<rot 0 -99.7377 -88.854>}]
 body [{shapeType=Capsule, length=263.7581, radius=133.1874, center=<51.1392 -0.2809 11.6985>, rotation=<rot 0 -90.1333 -85.4889>}]
 brakepad_bk_l [{shapeType=Capsule, length=16.1106, radius=12.4869, center=<-5.1624 -0.2527 4.3403>, rotation=<rot 0 90.0905 -137.4345>}]
 brakepad_bk_r [{shapeType=Capsule, length=16.1106, radius=12.4869, center=<-5.1624 0.2527 4.3403>, rotation=<rot 0 89.9095 -137.4345>}]
 en_low_panel_l [{shapeType=Capsule, length=15.7727, radius=14.677, center=<7.638 -0.7207 12.0492>, rotation=<rot 0 -2.1929 -89.0999>}]
 en_low_panel_r [{shapeType=Capsule, length=15.7727, radius=14.677, center=<7.638 0.7207 12.0492>, rotation=<rot 0 -177.8071 -89.0999>}]
 en_up_panel_l [{shapeType=Capsule, length=17.4569, radius=15.4923, center=<-15.4782 -0.2248 -8.4732>, rotation=<rot 0 89.7748 -56.5522>}]
 en_up_panel_r [{shapeType=Capsule, length=17.4569, radius=15.4923, center=<-15.4783 0.2248 -8.4732>, rotation=<rot 0 90.2252 -56.5522>}]
 glass_b [{shapeType=Capsule, length=70.6109, radius=27.1191, center=<65.3525 0 11.1676>, rotation=<rot 0 90 -81.4466>}]
 glass_c [{shapeType=Capsule, length=135.0049, radius=65.2818, center=<-106.7878 0 14.0723>, rotation=<rot 0 -90 -82.8756>}]
 shock_up_bk_l [{shapeType=Capsule, length=24.6417, radius=13.6995, center=<-5.7919 -1.0855 -2.4616>, rotation=<rot -27.0936 -90 -31.9195>}]
 wheel_disc_bk_l [{shapeType=Capsule, length=21.6264, radius=21.6264, center=<-0.0644 -2.1633 0.0502>, rotation=<rot 0 -90 0.0001>}]
 wheel_disc_bk_r [{shapeType=Capsule, length=21.6264, radius=21.6264, center=<-0.0644 2.1633 0.0502>, rotation=<rot 0 -90 0.0001>}]
 wheel_disc_fr_l [{shapeType=Capsule, length=21.6264, radius=21.6263, center=<-0.0535 -1.9829 -0.0801>, rotation=<rot 0 -90 -0.0002>}]
 wheel_disc_fr_r [{shapeType=Capsule, length=21.6264, radius=21.6263, center=<-0.0535 1.9829 -0.0801>, rotation=<rot 0 -90 0.0001>}]
 wheel_steering_fr_l [{shapeType=Capsule, length=28.6905, radius=16.9777, center=<-2.6114 14.6418 2.5692>, rotation=<rot 0 -21.0294 -70.9138>}]
 wheel_steering_fr_r [{shapeType=Capsule, length=16.1106, radius=12.4869, center=<-5.1514 3.6227 4.2104>, rotation=<rot 0 89.9095 -137.4345>}]
}
constraints [
 {bone1Name=glass_c, bone2Name=body, swing1LimitDegrees=45, swing2LimitDegrees=45, twistLimitDegrees=45}
 {bone1Name=glass_b, bone2Name=body, swing1LimitDegrees=45, swing2LimitDegrees=45, twistLimitDegrees=45}
 {bone1Name=shock_up_bk_l, bone2Name=body, swing1LimitDegrees=45, swing2LimitDegrees=45, twistLimitDegrees=45}
 {bone1Name=en_up_panel_l, bone2Name=body, swing1LimitDegrees=45, swing2LimitDegrees=45, twistLimitDegrees=45}
 {bone1Name=en_up_panel_r, bone2Name=body, swing1LimitDegrees=45, swing2LimitDegrees=45, twistLimitDegrees=45}
 {bone1Name=en_low_panel_l, bone2Name=body, swing1LimitDegrees=45, swing2LimitDegrees=45, twistLimitDegrees=45}
 {bone1Name=en_low_panel_r, bone2Name=body, swing1LimitDegrees=45, swing2LimitDegrees=45, twistLimitDegrees=45}
 {bone1Name=back_fender_l, bone2Name=body, swing1LimitDegrees=45, swing2LimitDegrees=45, twistLimitDegrees=45}
 {bone1Name=back_fender_r, bone2Name=body, swing1LimitDegrees=45, swing2LimitDegrees=45, twistLimitDegrees=45}
 {bone1Name=wheel_disc_fr_l, bone2Name=wheel_steering_fr_l, swing1LimitDegrees=45, swing2LimitDegrees=45, twistLimitDegrees=45}
 {bone1Name=wheel_steering_fr_r, bone2Name=axle_pivot_fr_r, swing1LimitDegrees=45, swing2LimitDegrees=45, twistLimitDegrees=45}
 {bone1Name=wheel_disc_fr_r, bone2Name=wheel_steering_fr_r, swing1LimitDegrees=45, swing2LimitDegrees=45, twistLimitDegrees=45}
 {bone1Name=wheel_disc_bk_l, bone2Name=axle_pivot_bk_l, swing1LimitDegrees=45, swing2LimitDegrees=45, twistLimitDegrees=45}
 {bone1Name=brakepad_bk_l, bone2Name=axle_pivot_bk_l, swing1LimitDegrees=45, swing2LimitDegrees=45, twistLimitDegrees=45}
 {bone1Name=wheel_disc_bk_r, bone2Name=axle_pivot_bk_r, swing1LimitDegrees=45, swing2LimitDegrees=45, twistLimitDegrees=45}
 {bone1Name=brakepad_bk_r, bone2Name=axle_pivot_bk_r, swing1LimitDegrees=45, swing2LimitDegrees=45, twistLimitDegrees=45}
]
